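# Unlit alpha-composite debug material that sphere-traces a distance field stored as a pseudo-volume (flipbook) texture inside a box.
# A Custom node intersects the camera ray with the local-space cube, clips it against scene depth and outputs entry position and thickness.
# A second Custom node steps by the selected channel's sampled distance minus DF Center, using a distance-scaled hit threshold.
# It returns inverse step count as shading and a hit mask.
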Material /Volumetrics/Content/RayMarching/Materials/M_DistanceField_Debug_PseudoVolume
graph {
 properties {bDisableDepthTest=true, blendMode=BLEND_AlphaComposite, pixelDepthOffsetMode=PDOM_Legacy, shadingModel=MSM_Unlit, usedShadingModels=Unlit}
 nodes [
  MaterialExpressionTextureObjectParameter_1 = TextureObjectParameter() {parameterName=RT, samplerType=SAMPLERTYPE_LinearColor, texture=/Volumetrics/Content/VolumeTextures/Textures/T_Diamond_01}
  MaterialExpressionScalarParameter_14 = ScalarParameter() {defaultValue=128, parameterName=MaxSteps}
  MaterialExpressionScalarParameter_12 = ScalarParameter() {defaultValue=1, parameterName=Plane Alignment}
  MaterialExpressionScalarParameter_5 = ScalarParameter() {defaultValue=128, parameterName=Steps}
  n1 = Custom!(PlaneAlignment=@MaterialExpressionScalarParameter_12, MaxSteps=@MaterialExpressionScalarParameter_5) {
   code="float localscenedepth = CalcSceneDepth(ScreenAlignedPosition(GetScreenPosition(Parameters)));\r\n\r\nfloat3 camerafwd = mul(float3(0.00000000,0.00000000,1.00000000),ResolvedView.ViewToTranslatedWorld);\r\n\r\nfloat3 depthpos = ((Parameters.CameraVector * localscenedepth) / abs( dot( camerafwd, Parameters.CameraVector ) ) );\r\n\r\ndepthpos = mul(depthpos,  (WSDemote(GetWorldToLocal(Parameters)))).xyz;\r\n\r\ndepthpos /= 256;\r\n\r\nlocalscenedepth = length(depthpos);\r\n\r\n//0-1\r\n//localscenedepth /= (GetPrimitiveData(Parameters).LocalObjectBoundsMax.x * 2 * scale);\r\n//localscenedepth /= abs( dot( camerafwd, Parameters.CameraVector ) );\r\n\r\n//bring vectors into local space to support object transforms\r\nfloat3 localcampos = mul(float4( DFDemote(ResolvedView.WorldCameraOrigin),1.00000000), (WSDemote(GetWorldToLocal(Parameters)))).xyz;\r\nfloat3 localcamvec = -normalize( mul(Parameters.CameraVector, WSDemote(GetWorldToLocal(Parameters))) );\r\n\r\n//make camera position 0-1\r\nlocalcampos = (localcampos / (GetPrimitiveData(Parameters).LocalObjectBoundsMax.x * 2)) + 0.5;\r\n\r\nfloat3 invraydir = 1 / localcamvec;\r\n\r\nfloat3 firstintersections = (0 - localcampos) * invraydir;\r\nfloat3 secondintersections = (1 - localcampos) * invraydir;\r\nfloat3 closest = min(firstintersections, secondintersections);\r\nfloat3 furthest = max(firstintersections, secondintersections);\r\n\r\nfloat t0 = max(closest.x, max(closest.y, closest.z));\r\nfloat t1 = min(furthest.x, min(furthest.y, furthest.z));\r\n\r\nfloat planeoffset = 1-frac( ( t0 - length(localcampos-0.5) ) * MaxSteps );\r\n\r\nt0 += (planeoffset / MaxSteps) * PlaneAlignment;\r\nt1 = min(t1, localscenedepth);\r\nt0 = max(0, t0);\r\n\r\nfloat boxthickness = max(0, t1 - t0);\r\n\r\nfloat3 entrypos = localcampos + (max(0,t0) * localcamvec);\r\n\r\n\r\nreturn float4( entrypos, boxthickness );"
   desc="\n*** Original source before expression upgrade ***\nfloat localscenedepth = CalcSceneDepth(ScreenAlignedPosition(GetScreenPosition(Parameters)));\r\n\r\nfloat3 camerafwd = mul(float3(0.00000000,0.00000000,1.00000000),ResolvedView.ViewToTranslatedWorld);\r\n\r\nfloat3 depthpos = ((Parameters.CameraVector * localscenedepth) / abs( dot( camerafwd, Parameters.CameraVector ) ) );\r\n\r\ndepthpos = mul(depthpos,  (GetPrimitiveData(Parameters).WorldToLocal)).xyz;\r\n\r\ndepthpos /= 256;\r\n\r\nlocalscenedepth = length(depthpos);\r\n\r\n//0-1\r\n//localscenedepth /= (GetPrimitiveData(Parameters).LocalObjectBoundsMax.x * 2 * scale);\r\n//localscenedepth /= abs( dot( camerafwd, Parameters.CameraVector ) );\r\n\r\n//bring vectors into local space to support object transforms\r\nfloat3 localcampos = mul(float4( ResolvedView.WorldCameraOrigin,1.00000000), (GetPrimitiveData(Parameters).WorldToLocal)).xyz;\r\nfloat3 localcamvec = -normalize( mul(Parameters.CameraVector, GetPrimitiveData(Parameters).WorldToLocal) );\r\n\r\n//make camera position 0-1\r\nlocalcampos = (localcampos / (GetPrimitiveData(Parameters).LocalObjectBoundsMax.x * 2)) + 0.5;\r\n\r\nfloat3 invraydir = 1 / localcamvec;\r\n\r\nfloat3 firstintersections = (0 - localcampos) * invraydir;\r\nfloat3 secondintersections = (1 - localcampos) * invraydir;\r\nfloat3 closest = min(firstintersections, secondintersections);\r\nfloat3 furthest = max(firstintersections, secondintersections);\r\n\r\nfloat t0 = max(closest.x, max(closest.y, closest.z));\r\nfloat t1 = min(furthest.x, min(furthest.y, furthest.z));\r\n\r\nfloat planeoffset = 1-frac( ( t0 - length(localcampos-0.5) ) * MaxSteps );\r\n\r\nt0 += (planeoffset / MaxSteps) * PlaneAlignment;\r\nt1 = min(t1, localscenedepth);\r\nt0 = max(0, t0);\r\n\r\nfloat boxthickness = max(0, t1 - t0);\r\n\r\nfloat3 entrypos = localcampos + (max(0,t0) * localcamvec);\r\n\r\n\r\nreturn float4( entrypos, boxthickness );"
   description=Ray March Cube Setup
   outputType=CMOT_Float4
   inputs [
    {
     inputName=PlaneAlignment
     input {expression=&self:MaterialExpressionScalarParameter_12}
    }
    {
     inputName=MaxSteps
     input {expression=&self:MaterialExpressionScalarParameter_5}
    }
   ]
  }
  MaterialExpressionMultiply_0 = Multiply(A=ComponentMask(None=@n1) {b=true, g=true, r=true}) {desc=Ray Entry}
  MaterialExpressionComponentMask_1 = ComponentMask(None=@n1) {a=true}
  MaterialExpressionScalarParameter_13 = ScalarParameter() {parameterName=DF Center}
  n2 = VectorParameter() {parameterName=Channel, defaultValue=<rgba 1 0 0 0>}
  MaterialExpressionAppendVector_1 = AppendVector(A=@n2.RGB, B=@n2.RGB)
  n3 = VectorParameter() {parameterName=XYFrames, defaultValue=<rgba 16 8 0 0>}
  MaterialExpressionAppendVector_0 = AppendVector(A=@n3.R, B=@n3.R)
  n4 = Custom!(
   DistanceField=@MaterialExpressionTextureObjectParameter_1
   MaxSteps=@MaterialExpressionScalarParameter_14
   CurPos=@MaterialExpressionMultiply_0
   MaxDist=@MaterialExpressionComponentMask_1
   DFCenter=@MaterialExpressionScalarParameter_13
   Channel=@MaterialExpressionAppendVector_1
   XYFrames=@MaterialExpressionAppendVector_0
  ) {
   code="float3 localcamvec = normalize( mul(Parameters.CameraVector, WSDemote(GetWorldToLocal(Parameters))) );\r\n\r\nconst float numFrames = XYFrames.x * XYFrames.y;\r\n\r\nfloat curdensity = 0;\r\nfloat totaldist=  0;\r\n\r\nfloat3 VolumeRes = numFrames;\r\nfloat3 RayPos = CurPos;\r\n\r\nRayPos -= 0.5;\r\nRayPos *= ( VolumeRes - 1) / (VolumeRes);\r\nRayPos += 0.5;\r\n\r\n\r\nfor (int i = 0; i < MaxSteps; i++)\r\n{\t\r\n\r\n\tfloat4 cursample = PseudoVolumeTexture(DistanceField, DistanceFieldSampler, RayPos, XYFrames, numFrames);\r\n\tcurdensity = dot (cursample, Channel);\r\n\tcurdensity = curdensity - DFCenter;\r\n\t\r\n\ttotaldist += curdensity;\r\n\t\r\n\tfloat e = totaldist * 0.001;\r\n\t\r\n\t//Sample Light Absorption and Scattering\r\n\tif( curdensity < e)\r\n\t{\r\n\t\treturn float2(1.0 / ((float) i + 1.0), 1);\r\n\t}\r\n\t\r\n\tRayPos -= localcamvec * curdensity;\r\n\t\r\n\tif (RayPos.x < 0 || RayPos.x > 1 || RayPos.y < 0 || RayPos.y > 1 || RayPos.z < 0 || RayPos.z > 1 || totaldist > MaxDist)\r\n\t{\r\n\t\treturn 0;\r\n\t}\r\n}\r\n\r\n\r\n\r\nreturn 0;\r\n"
   desc="\n*** Original source before expression upgrade ***\nfloat3 localcamvec = normalize( mul(Parameters.CameraVector, GetPrimitiveData(Parameters).WorldToLocal) );\r\n\r\nconst float numFrames = XYFrames.x * XYFrames.y;\r\n\r\nfloat curdensity = 0;\r\nfloat totaldist=  0;\r\n\r\nfloat3 VolumeRes = numFrames;\r\nfloat3 RayPos = CurPos;\r\n\r\nRayPos -= 0.5;\r\nRayPos *= ( VolumeRes - 1) / (VolumeRes);\r\nRayPos += 0.5;\r\n\r\n\r\nfor (int i = 0; i < MaxSteps; i++)\r\n{\t\r\n\r\n\tfloat4 cursample = PseudoVolumeTexture(DistanceField, DistanceFieldSampler, RayPos, XYFrames, numFrames);\r\n\tcurdensity = dot (cursample, Channel);\r\n\tcurdensity = curdensity - DFCenter;\r\n\t\r\n\ttotaldist += curdensity;\r\n\t\r\n\tfloat e = totaldist * 0.001;\r\n\t\r\n\t//Sample Light Absorption and Scattering\r\n\tif( curdensity < e)\r\n\t{\r\n\t\treturn float2(1.0 / ((float) i + 1.0), 1);\r\n\t}\r\n\t\r\n\tRayPos -= localcamvec * curdensity;\r\n\t\r\n\tif (RayPos.x < 0 || RayPos.x > 1 || RayPos.y < 0 || RayPos.y > 1 || RayPos.z < 0 || RayPos.z > 1 || totaldist > MaxDist)\r\n\t{\r\n\t\treturn 0;\r\n\t}\r\n}\r\n\r\n\r\n\r\nreturn 0;\r\n"
   description=DistanceField Debug View-PseudoVolume Texture
   outputType=CMOT_Float2
   inputs [
    {
     inputName=DistanceField
     input {expression=&self:MaterialExpressionTextureObjectParameter_1}
    }
    {
     inputName=MaxSteps
     input {expression=&self:MaterialExpressionScalarParameter_14}
    }
    {
     inputName=CurPos
     input {expression=&self:MaterialExpressionMultiply_0}
    }
    {
     inputName=MaxDist
     input {expression=&self:MaterialExpressionComponentMask_1}
    }
    {
     inputName=DFCenter
     input {expression=&self:MaterialExpressionScalarParameter_13}
    }
    {
     inputName=Channel
     input {expression=&self:MaterialExpressionAppendVector_1}
    }
    {
     inputName=XYFrames
     input {expression=&self:MaterialExpressionAppendVector_0}
    }
   ]
  }
  ComponentMask(None=@n4) {r=true}
  ComponentMask(None=@n4) {g=true}
 ]
}
